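# Stereo one-shot MetaSound source for PistolCanary MaskMaker event audio, playing a wave asset supplied through an input.
# A stereo Wave Player starts on play from the given start time, and each channel is multiplied by a Master Volume input before reaching the outputs.
# The player's finish trigger ends the one-shot.
# It uses the MaskMaker far attenuation asset and the Fort events gameplay sound class.
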MetaSoundSource /PistolCanary_Event_Audio/Assets/MSS_PistolCanary_MaskMakerOneShot_Stereo
inputs [
 Master Volume {type=Float}
 UE.Source.OnPlay {type=Trigger}
 UE.Source.StartTime.StartTime {type=Time}
 Wave Asset {type=WaveAsset}
]
interfaces [UE.OutputFormat.Stereo:1.0, UE.Source.OneShot:1.0, UE.Source.StartTime:1.0, UE.Source:1.1]
outputs [
 "UE.OutputFormat.Stereo.Audio:0" {type=Audio}
 "UE.OutputFormat.Stereo.Audio:1" {type=Audio}
 UE.Source.OneShot.OnFinished {type=Trigger}
]
pages [Default {
  input_defaults {Master Volume=1, UE.Source.OnPlay=null, UE.Source.StartTime.StartTime=0, Wave Asset=null}
  nodes [
   WavePlayer_0 = UE.Wave Player:Stereo(Play=^UE.Source.OnPlay, Start Time=^UE.Source.StartTime.StartTime, Wave Asset=^Wave Asset)
   Multiply_0 = UE.Multiply:Audio by Float(AdditionalOperands=^Master Volume, PrimaryOperand=@WavePlayer_0.Out Right)
   Multiply_1 = UE.Multiply:Audio by Float(AdditionalOperands=^Master Volume, PrimaryOperand=@WavePlayer_0.Out Left)
  ]
  links [
   {dst="@UE.OutputFormat.Stereo.Audio:1", src=Multiply_0.Out}
   {dst="@UE.OutputFormat.Stereo.Audio:0", src=Multiply_1.Out}
   {dst="@UE.Source.OneShot.OnFinished", src=WavePlayer_0.On Finished}
  ]
 }]
properties {
 assetImportData=&self:AssetImportData
 attenuationSettings=/PistolCanary_Audio/Assets/Settings/ATT_PistolCanary_MaskMaker_Far
 maxDistance=23000
 numChannels=2
 outputFormat=Stereo
 sampleRate=48000
 soundClassObject=&Fort_Events_Gameplay
}
